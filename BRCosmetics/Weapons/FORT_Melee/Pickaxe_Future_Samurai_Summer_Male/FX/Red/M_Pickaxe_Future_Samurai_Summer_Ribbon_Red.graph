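# Translucent, lit smoke ribbon material for Niagara trails on the red Future Samurai Summer pickaxe.
# Panning base and wisp smoke textures, timed by game time or a dynamic parameter, are summed and square-rooted into a shared mask.
# That mask drives a color gradient tinted by particle color for emissive, plus an alpha with edge fades, non-mobile depth fade and time-of-day scaling.
# A normal map pans with the base smoke.
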Material /BRCosmetics/Weapons/FORT_Melee/Pickaxe_Future_Samurai_Summer_Male/FX/Red/M_Pickaxe_Future_Samurai_Summer_Ribbon_Red
graph {
 properties {
  bUsedWithNiagaraRibbons=true
  blendMode=BLEND_Translucent
  pixelDepthOffsetMode=PDOM_Legacy
  translucencyLightingMode=TLM_VolumetricPerVertexDirectional
  usedShadingModels=Default Lit
 }
 nodes [
  n1 = TextureCoordinate() {coordinateIndex=1}
  n2 = LinearInterpolate(
   A=VectorParameter() {parameterName=Base Color A, defaultValue=<rgba 0.5 0.5 0.5 1>}.RGB
   B=VectorParameter() {parameterName=Base Color B, defaultValue=<rgba 5 5 5 1>}.RGB
   Alpha=Power(
    Base=MaterialFunctionCall(Float2=@n1) {materialFunction=&BreakOutFloat2Components}.R
    Exp=ScalarParameter() {defaultValue=2, parameterName=Color Power}
   )
  )
  n3 = Multiply(
   A=TextureCoordinate() {coordinateIndex=1}
   B=ComponentMask(None=VectorParameter() {group=Smoke, parameterName=Smoke UV Scale, sortPriority=7, defaultValue=<rgba 1 1 0 1>}.RGB) {g=true, r=true}
  )
  n4 = VectorParameter() {group=Smoke, parameterName=Wisp UV Scale Multiplier, sortPriority=8, defaultValue=<rgba 0.0618 1 0 1>}
  n5 = ComponentMask(None=@n4.RGB) {g=true, r=true}
  n6 = ScalarParameter() {defaultValue=0.2, group=Smoke, parameterName=Wisp Panner Speed, sortPriority=5}
  n7 = Constant()
  n8 = LinearInterpolate(
   A=Time()
   B=DynamicParameter() {paramNames [Time,Random,ColorRange,Alpha]}.Time
   Alpha=ScalarParameter() {defaultValue=1, group=Smoke, parameterName=Use Dynamic Parameter Controlled Time, sortPriority=6}
  )
  n9 = TextureSampleParameter2D(UVs=Panner(Coordinate=Multiply(A=@n3, B=@n5) {constB=0.618}, Time=@n8, Speed=AppendVector(A=@n6, B=@n7))) {group=Smoke, parameterName=Smoke Wisp, sortPriority=2, texture=/Game/Effects/Fort_Effects/Textures/Smoke/T_Smoke_Trail_Soft_2}
  n10 = Panner(
   Coordinate=@n3
   Time=@n8
   Speed=AppendVector(A=ScalarParameter() {defaultValue=0.4, group=Smoke, parameterName=Smoke Base Panner Speed, sortPriority=4}, B=Constant())
  )
  n11 = TextureSampleParameter2D(UVs=@n10) {
   group=Smoke
   parameterName=Smoke Base
   samplerType=SAMPLERTYPE_Grayscale
   texture=&T_Smoke_Trail_Soft
  }
  n12 = SquareRoot(None=Add(
    A=ChannelMaskParameter(None=@n9.RGBA) {group=Smoke, parameterName=Smoke Wisp Channel Mask, sortPriority=3}.R
    B=ChannelMaskParameter(None=@n11.RGBA) {group=Smoke, parameterName=Smoke Base Channel Mask, sortPriority=1}.R
   ))
  n13 = Saturate(None=Multiply(A=@n12, B=ScalarParameter() {defaultValue=0.5, group=Smoke, parameterName=Smoke Intensity, sortPriority=9}) {constB=0.5})
  MaterialFunctionCall(Emissive=Multiply(A=Multiply(A=@n2, B=@n13), B=ParticleColor().RGB)) {materialFunction=&MF_FX_EmissiveAdjustments}
  n14 = ComponentMask(None=TextureCoordinate()) {r=true}
  n15 = ScalarParameter() {defaultValue=-0.05, group="Fade Leading && Trailing Edges", parameterName=Leading Edge Fade Bias, sortPriority=2}
  n16 = Saturate(None=Multiply(A=Add(A=@n14, B=@n15), B=ScalarParameter() {defaultValue=2, group="Fade Leading && Trailing Edges", parameterName=Leading Edge Fade Scale}) {constB=3})
  n17 = ScalarParameter() {defaultValue=1, group="Fade Leading && Trailing Edges", parameterName=Fade Leading Edge, sortPriority=4}
  n18 = OneMinus(None=@n14)
  n19 = ScalarParameter() {defaultValue=-0.05, group="Fade Leading && Trailing Edges", parameterName=Trailing Edge Fade Bias, sortPriority=3}
  n20 = Saturate(None=Multiply(
    A=Add(A=@n18, B=@n19)
    B=ScalarParameter() {defaultValue=2, group="Fade Leading && Trailing Edges", parameterName=Trailing Edge Fade Scale, sortPriority=1}
   ) {constB=2})
  n21 = ScalarParameter() {defaultValue=1, group="Fade Leading && Trailing Edges", parameterName=Fade Trailing Edge, sortPriority=5}
  n22 = Multiply(
   A=Multiply(A=DynamicParameter() {paramNames [Time,Random,ColorRange,Alpha]}.Alpha, B=ParticleColor().A)
   B=Multiply(A=LinearInterpolate(B=@n16, Alpha=@n17) {constA=1}, B=LinearInterpolate(B=@n20, Alpha=@n21) {constA=1})
  )
  n23 = ShadingPathSwitch(Default=DepthFade(Opacity=@n22) {fadeDistanceDefault=25}, Mobile=@n22) {desc=Migrated from FeatureLevelSwitch}
  Multiply(
   A=Subtract(A=Multiply(A=@n12, B=@n23)) {constB=0.02}
   B=MaterialFunctionCall(
    Color_Morning=ScalarParameter() {defaultValue=1, group=Color, parameterName=Morning Intensity, sortPriority=3}
    Color_Day=ScalarParameter() {defaultValue=1, group=Color, parameterName=Day Intensity, sortPriority=4}
    Color_Evening=ScalarParameter() {defaultValue=1, group=Color, parameterName=Evening Intensity, sortPriority=5}
    Color_Night=ScalarParameter() {defaultValue=1, group=Color, parameterName=Night Intensity, sortPriority=6}
   ) {materialFunction=&Color_Time_of_Day}.Result
  )
  TextureSampleParameter2D(UVs=@n10) {
   group=Smoke
   parameterName=Smoke Base Normal
   samplerType=SAMPLERTYPE_Normal
   sortPriority=10
   texture=/Game/Effects/Fort_Effects/Textures/Smoke/T_Smoke_Trail_Soft_N
  }
 ]
}
